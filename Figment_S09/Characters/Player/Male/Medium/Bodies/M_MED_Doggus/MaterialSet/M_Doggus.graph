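# Skeletal-mesh material for the Doggus player body, built on material attributes.
# Texture parameters feed the surface, with one spec-mask channel driving metallic, specular and roughness, and emissive scaled by an MPC_Doggus parameter.
# Attributes chain through MF_DoggusDam, MF_DoggusWet and MF_ZeroPointCharge, while vertex-color element hiding supplies world position offset.
# A distorted, panning electric-noise energy network sits in the graph with its final multiply unconnected.
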Material /Figment_S09/Characters/Player/Male/Medium/Bodies/M_MED_Doggus/MaterialSet/M_Doggus
graph {
 properties {
  bUseMaterialAttributes=true
  bUsedWithMeshDeformer=true
  bUsedWithSkeletalMesh=true
  pixelDepthOffsetMode=PDOM_Legacy
  usedShadingModels=Default Lit
 }
 nodes [
  ActorPositionWS()
  Constant() {r=2}
  Constant() {r=7}
  Constant() {r=0.3}
  Constant() {r=6}
  n1 = TextureSampleParameter2D() {
   parameterName=SpecTexture
   samplerType=SAMPLERTYPE_Masks
   texture=/Game/Characters/Player/Male/Medium/Bodies/M_MED_Doggus/Textures/T_Doggus_Torso_S
  }
  n2 = TextureSampleParameter2D() {
   parameterName=NormalTexture
   samplerType=SAMPLERTYPE_Normal
   texture=/Game/Characters/Player/Male/Medium/Bodies/M_MED_Doggus/Textures/T_Doggus_Torso_N
  }
  n3 = TextureSampleParameter2D() {parameterName=EmissiveTexture, texture=/Game/Characters/Player/Male/Medium/Bodies/M_MED_Doggus/Textures/T_Doggus_Torso_E}
  n4 = VectorParameter() {parameterName=EmissiveColor, defaultValue=<rgba 0 3 5 1>}
  n5 = MakeMaterialAttributes!(
   BaseColor=TextureSampleParameter2D() {parameterName=DiffuseTexture, texture=/Game/Characters/Player/Male/Medium/Bodies/M_MED_Doggus/Textures/T_Doggus_Torso_D}.RGB
   Metallic=@n1.G
   Specular=@n1.G
   Roughness=@n1.G
   Anisotropy
   EmissiveColor=Multiply(
    A=Multiply(A=@n3.RGB, B=@n4.RGB)
    B=CollectionParameter() {collection=/Game/Characters/Player/Male/Medium/Bodies/M_MED_Doggus/MaterialSet/MPC_Doggus, parameterName=DoggusEmissive}
   )
   Opacity
   OpacityMask
   Normal=@n2.RGB
   Tangent
   WorldPositionOffset=MaterialFunctionCall!(WPO) {materialFunction=/Game/Packages/Fortress_SharedMaterials/Base_Material_Functions/Fortnite_HideElementsWithVertexColor}.Result
   SubsurfaceColor
   ClearCoat
   ClearCoatRoughness
   AmbientOcclusion
   Refraction
   CustomizedUVs_0
   CustomizedUVs_1
   CustomizedUVs_2
   CustomizedUVs_3
   CustomizedUVs_4
   CustomizedUVs_5
   CustomizedUVs_6
   CustomizedUVs_7
   PixelDepthOffset
   ShadingModel
   Displacement
  )
  MaterialFunctionCall!(
   In=MaterialFunctionCall!(
    AO_Mask_Input
    In=MaterialFunctionCall!(In=@n5) {materialFunction=&/MF_DoggusDam}.Result
   ) {materialFunction=/Game/Characters/Player/Male/Medium/Bodies/M_MED_Doggus/MaterialSet/MF_DoggusWet}.Result
   Mask
   EnergyDirection
   EnergySubtractMask
   EnergyBrightness
   EnergyProgress
  ) {materialFunction=/Game/Characters/Player/Male/Medium/Bodies/M_MED_Doggus/MaterialSet/MF_ZeroPointCharge}
  n6 = Multiply(A=TextureCoordinate(), B=ScalarParameter() {defaultValue=2, parameterName=Energy_Tile1})
  n7 = Time()
  n8 = VectorParameter() {parameterName=Energy_Speed, defaultValue=<rgba 0.4 0.4 0 0>}
  n9 = MaterialFunctionCall(X=@n8.R, Y=@n8.R) {materialFunction=&MakeFloat2}
  n10 = TextureCoordinate()
  n11 = ScalarParameter() {defaultValue=8, parameterName=Energy_DistortTile}
  n12 = Time()
  n13 = MaterialFunctionCall(X=ScalarParameter() {defaultValue=-0.2, parameterName=Energy_DistortSpeed}, Y=ScalarParameter() {parameterName=Energy_DistortSpeed2}) {materialFunction=&MakeFloat2}
  n14 = TextureSample(UVs=Add(A=Multiply(A=@n10, B=@n11), B=Multiply(A=@n12, B=@n13.Result))) {samplerType=SAMPLERTYPE_LinearColor, texture=/Game/Missions/Primary/DeliverTheBomb/Textures/T_CloudNoise}
  n15 = Multiply(A=Subtract(A=AppendVector(A=@n14.R, B=@n14.R)) {constB=0.5}, B=ScalarParameter() {defaultValue=0.2, parameterName=Energy_DistortAmount}) {constB=0.2}
  n16 = TextureSample(UVs=Add(A=Panner(Coordinate=@n6, Time=@n7, Speed=@n9.Result), B=@n15)) {samplerType=SAMPLERTYPE_LinearGrayscale, texture=/Game/Characters/Player/Male/Medium/Bodies/M_MED_SRDrift/Textures/T_ElectricNoise}
  n17 = Multiply(A=TextureCoordinate(), B=ScalarParameter() {defaultValue=9, parameterName=Energy_Tile2})
  n18 = MaterialFunctionCall(X=@n8.B, Y=@n8.B) {materialFunction=&MakeFloat2}
  n19 = TextureSample(UVs=Add(A=@n15, B=Panner(Coordinate=@n17, Time=@n7, Speed=@n18.Result))) {samplerType=SAMPLERTYPE_LinearGrayscale, texture=&T_ElectricNoise}
  n20 = SquareRoot(None=Multiply(A=Multiply(A=@n16.R, B=@n19.R), B=ScalarParameter() {defaultValue=2, parameterName=Energy_Mod}) {constB=2})
  n21 = Power(Base=@n20, Exp=ScalarParameter() {defaultValue=6, parameterName=Energy_Power})
  n22 = Multiply(
   A=Power(Base=@n20, Exp=ScalarParameter() {defaultValue=3, parameterName=Energy_Power2})
   B=ScalarParameter() {defaultValue=0.02, parameterName=Energy_SoftMod}
  )
  Multiply(A=Multiply(A=VectorParameter() {parameterName=Energy_Color, defaultValue=<rgba 0 33.977 50 1>}.RGB, B=Add(A=@n21, B=@n22)))
  TextureSample() {samplerType=SAMPLERTYPE_LinearGrayscale, texture=&T_ElectricNoise}
  TextureSample() {texture=/Game/Effects/Fort_Effects/Textures/Plasma/T_Turbulence_Seamless}
  TextureSample() {samplerType=SAMPLERTYPE_LinearColor, texture=&T_CloudNoise}
  TextureSample() {samplerType=SAMPLERTYPE_LinearColor, texture=/Game/Effects/Fort_Effects/Textures/Tile/LinerNoise64x64}
 ]
}
